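# Weapon reticle extension widget for CoffeePuff that shows whether the targeted actor is environment or an enemy.
# On construct it binds to the equipped weapon's CoffeePuff targeting feedback component, unbinding on destruct.
# Target type changes play the matching in-range animation, or reverse the previous one when the target clears.
# Animations drive material parameters on one reticle image; a hookshot animation exists but is not played in these graphs.
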WidgetBlueprint /CoffeePuffGameplay/UI/WBP_CoffeePuff_Reticle
animations [
 Anim_Hookshot {
  binding_count=1
  length_seconds=1.3333
  object=&self:Anim_Hookshot
  sequence {
   bindings [Image_Reticle {
     id=#1
     possessed_class=Image
     tracks [{
       type=MovieSceneWidgetMaterialTrack
       sections [{
         type=MovieSceneParameterSection
         channels [HookShot Scale {type=MovieSceneScriptingFloatChannel, keys [[0.0005,-0.0019],[33.4885,-0.994],[40,-0.0019]]}]
        }]
      }]
    }]
   display_rate {numerator=30}
  }
 }
 Anim_EnvironmentInRange {
  binding_count=1
  length_seconds=0.1025
  object=&self:Anim_EnvironmentInRange
  sequence {bindings [Image_Reticle {
     id=#2
     possessed_class=Image
     tracks [{
       type=MovieSceneWidgetMaterialTrack
       sections [{
         type=MovieSceneParameterSection
         end_frame=2
         start_frame=0
         channels [
          Reticle Master Opacity {type=MovieSceneScriptingFloatChannel, keys [[0,0],[2.0437,1]]}
          In Range Spike Offset {type=MovieSceneScriptingFloatChannel, keys [[0,0],[2.0437,1]]}
          Outer Bracket Spike Opacity Switch {type=MovieSceneScriptingFloatChannel, keys [[0,0],[0.4057,1],[2.0493,1]]}
          Enemy In Range {type=MovieSceneScriptingFloatChannel, const [0,2]}
          Spikes Color Switch {type=MovieSceneScriptingFloatChannel, const [0,2]}
         ]
        }]
      }]
    }]}
 }
 Anim_EnemyInRange {
  binding_count=1
  length_seconds=0.1028
  object=&self:Anim_EnemyInRange
  sequence {bindings [Image_Reticle {
     id=#3
     possessed_class=Image
     tracks [{
       type=MovieSceneWidgetMaterialTrack
       sections [{
         type=MovieSceneParameterSection
         end_frame=2
         start_frame=0
         channels [
          Reticle Master Opacity {type=MovieSceneScriptingFloatChannel, keys [[0,0],[2.0553,1]]}
          Enemy In Range {type=MovieSceneScriptingFloatChannel, keys [[0,0],[2.0553,1]]}
          Spikes Color Switch {type=MovieSceneScriptingFloatChannel, keys [[0,0],[2.0553,1]]}
          In Range Spike Offset {type=MovieSceneScriptingFloatChannel, keys [[0,0],[2.0553,1]]}
          Outer Bracket Spike Opacity Switch {type=MovieSceneScriptingFloatChannel, keys [[0,0],[0.4073,1],[2.0553,1]]}
         ]
        }]
      }]
    }]}
 }
]
graphs [
 EventGraph {nodes [
   n1 = HandleTargetingActorTypeChanged!() -> (OutputDelegate, then, ActorType:AbilityTargetingTargetActorTypeEnum Enum=NewEnumerator0) {is=title K2Node_CustomEvent}
   > UpdateTargetingState!(execute, self, NewTargetType:AbilityTargetingTargetActorTypeEnum Enum=@n1.ActorType) -> (then)
   Event Construct()
   > Parent: Construct()
   n2 = GetOwningPlayer()
   > Is Valid(InputObject=@n2)
   > n3 = Cast To PlayerPawn_Athena(Object=GetViewTarget!(self:Controller Object Reference=@n2) -> (ReturnValue:Actor Object Reference))
   > Set ViewTargetPlayerPawn(ViewTargetPlayerPawn:Player Pawn Athena Object Reference=@n3) -> (Output_Get:Player Pawn Athena Object Reference)
   n4 = GetEquippedWeapon!(self:Fort Weapon Reticle Extension Widget Base Object Reference) -> (ReturnValue:Fort Weapon Object Reference)
   > n5 = Set EquippedWeapon(EquippedWeapon:Fort Weapon Object Reference=@n4) -> (Output_Get:Fort Weapon Object Reference)
   > Is Valid(InputObject=@n5)
   n6 = GetComponentByClass!(
    self:Actor Object Reference=@n4
    ComponentClass:Actor Component Class Reference=/CoffeePuffGameplay/Gameplay/B_CoffeePuffTargetingFeedbackComponent.B_CoffeePuffTargetingFeedbackComponent_C
   ) -> (ReturnValue:B Coffee Puff Targeting Feedback Component Object Reference)
   > Is Valid(InputObject=@n6)
   > Bind Event to On Targeted Actor Type Changed(self:B Ability Targeting Feedback Component Parent Object Reference=@n6, Delegate=Create Event()) {is=title K2Node_AddDelegate}
   Event Destruct()
   > n7 = Sequence()
   > n8 = Get!(execute:Exec) -> (then:Exec, else:Exec, EquippedWeapon:Fort Weapon Object Reference)
   n9 = GetComponentByClass(self=@n8, ComponentClass=&B_CoffeePuffTargetingFeedbackComponent_C)
   > Is Valid(InputObject=@n9)
   > Unbind Event from On Targeted Actor Type Changed(self:B Ability Targeting Feedback Component Parent Object Reference=@n9, Delegate=Create Event()) {is=title K2Node_RemoveDelegate}
   Parent: Destruct(execute=@n7.then_1)
  ]}
 UpdateTargetingState {nodes [
   n1 = UpdateTargetingState!() -> (then, NewTargetType:AbilityTargetingTargetActorTypeEnum Enum=NewEnumerator0) {is=title K2Node_FunctionEntry}
   > n2 = Branch(Condition="Equal (Enum)"(
     A:AbilityTargetingTargetActorTypeEnum Enum=@n1
     B:AbilityTargetingTargetActorTypeEnum Enum=Get PreviousTarget() -> (PreviousTarget:AbilityTargetingTargetActorTypeEnum Enum=NewEnumerator0)
    ))
   StopAllAnimations(execute=@n2.else, self:User Widget Object Reference)
   > n3 = Switch on AbilityTargetingTargetActorTypeEnum!(
    execute
    Selection:AbilityTargetingTargetActorTypeEnum Enum=Get NewTargetType() -> (NewTargetType:AbilityTargetingTargetActorTypeEnum Enum=NewEnumerator0)
   ) -> (NewEnumerator0:Exec, NewEnumerator1:Exec, NewEnumerator2:Exec) {is=title K2Node_SwitchEnum}
   PlayAnimation(execute=@n3.NewEnumerator1, InAnimation=Get Anim_EnvironmentInRange() -> (Anim_EnvironmentInRange:Widget Animation Object Reference))
   > Set PreviousTarget(PreviousTarget:AbilityTargetingTargetActorTypeEnum Enum=NewEnumerator1) -> (Output_Get:AbilityTargetingTargetActorTypeEnum Enum=NewEnumerator0)
   PlayAnimationReverse(
    execute=@n3.NewEnumerator0
    InAnimation=Select!(
     NewEnumerator0:Widget Animation Object Reference
     NewEnumerator1:Widget Animation Object Reference=Get Anim_EnvironmentInRange()
     NewEnumerator2:Widget Animation Object Reference=Get Anim_EnemyInRange() -> (Anim_EnemyInRange:Widget Animation Object Reference)
     Index:AbilityTargetingTargetActorTypeEnum Enum=Get PreviousTarget() -> (PreviousTarget=NewEnumerator0)
    ) -> (ReturnValue:Widget Animation Object Reference)
   )
   > Set PreviousTarget(PreviousTarget=NewEnumerator0) -> (Output_Get=NewEnumerator0)
   PlayAnimation(execute=@n3.NewEnumerator2, InAnimation=Get Anim_EnemyInRange())
   > Set PreviousTarget(PreviousTarget=NewEnumerator2) -> (Output_Get=NewEnumerator0)
  ]}
]
variables [
 EquippedWeapon {type=FortWeapon, default=null}
 PreviousTarget {type=AbilityTargetingTargetActorTypeEnum, default=None}
 ViewTargetPlayerPawn {type=PlayerPawn_Athena_C}
]
widgets {
 info {
  inheritedWidgetCount=1
  parentClass=/Script/FortniteUI.FortWeaponReticleExtensionWidgetBase
  rootWidgetClass=&Overlay
  widgetCount=3
 }
 widgets [
  Overlay_36 {widgetClassPath=&Overlay, widget=Overlay_36}
  Image_Reticle {
   widgetClassPath=&Image
   bIsVariable=true
   parent=Overlay_36
   widget=Image_Reticle
   properties {brush {
     bIsDynamicallyLoaded=false
     drawAs=Image
     imageType=FullColor
     mirroring=NoMirror
     resourceName=None
     resourceObject=/CoffeePuffGameplay/UI/Materials/MI_UI_Ret_CoffeePuff
     tiling=NoTile
     imageSize=<92 92>
     margin=<ltrb 0 0 0 0>
     outlineSettings {
      bUseBrushTransparency=false
      roundingType=HalfHeightRadius
      width=0
      color {colorUseRule=UseColor_Specified, specifiedColor=<rgba 0 0 0 0>}
      cornerRadii=<0 0 0 0>
     }
     tintColor {colorUseRule=UseColor_Specified, specifiedColor=<rgba 1 1 1 1>}
     uVRegion {bIsValid=false, max=<0 0>, min=<0 0>}
    }}
   slot {class=OverlaySlot}
  }
  TutorialHighlight {widgetClassPath=&TutorialHighlight}
 ]
}
